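; Halloween event prop that may spawn a damageable object as haunted, counting toward a quest when destroyed.
; When the playlist is ready, the server rolls haunting, then spawns the object floating, or in place or not at all if unhaunted.
; RepNotifies drive setup: proxy and ghost meshes copy the hidden object's mesh, with dynamic materials and particles for spooky bobbing.
; Death sends the quest stat, triggers client destruction FX and destroys the actor.
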(event EventBeginPlay
  (bind _spawneddamageableobject (Variables|Default|GetSpawnedDamageableObject))
  (bind _returnvalue (Ability|GetAbilitySystemComponent _spawneddamageableobject))
  (Networking|SwitchHasAuthority
    (:Authority
      (Utilities|FlowControl|Sequence
        (:then_0
          ;[1 1 1] Not Haunted if event not active
          (Class|FortAsyncActionCurrentPlaylistReady|PlaylistDataReadyAsync
            (:OnReady
              (bind _biseventactive (CalendarEvents|GetCalendarEventInformation (Variables|Default|GetHalloweenEventFlag)))
              (if _biseventactive
                ;[1 1 1] let the server pick if this is a haunted item or not
                (Variables|Default|SetIsHaunted (or (Variables|Default|GetbForceHaunted) (>= (Variables|Default|GetHauntingChance) (Math|Random|RandomFloatinRange 0.0 99.0))))
                (else
                  (Variables|Default|SetIsHaunted false))))))
        (:then_1
          ;[1 1 1] Only let Haunted Object stick around if Event is Active
          (MCPContext|WatchCalendarEvent (Variables|Default|GetHalloweenEventFlag)
            (:then
              (if (Variables|Default|GetIsHaunted)
                ;[1 1 1] Apply Haunted Item tag to spawned object via GE
                (Utilities|IsValid _returnvalue
                  (:Is%20Valid
                    (GameplayEffects|ApplyGameplayEffectToSelf :self _returnvalue :GameplayEffectClass "/Game/Athena/Items/QuestInteractables/HauntedObjects/Blueprints/GE_HauntedItem.GE_HauntedItem_C" :Level 0.0)))
                (else
                  _)))
            (:OnEventActive
              _)))
        (:then_2
          (if (not true)
            (Utilities|FlowControl|Delay 5.0))))
      (Utilities|FlowControl|Gate false))))

(event Custom|CreateChildProp
  (bind _self self)
  (bind _objectbp (Variables|Default|GetObjectBP))
  (if (Variables|Default|GetIsHaunted)
    ;[1 1 1] Spawn Damageable Object in place, or hovering in the air if haunted
    (bind _hauntedtransform (CallFunction|GetHauntedTransform))
    (bind _returnvalue (Game|SpawnActorfromClass _objectbp _hauntedtransform "Undefined" "SelectDefaultAtRuntime" _self 0 0 0))
    (Variables|Default|SetSpawnedDamageableObject _returnvalue)
    (else
      (bind _returnvalue_2 (Game|SpawnActorfromClass _objectbp (Transformation|GetActorTransform) "Undefined" "SelectDefaultAtRuntime" _self 0 0 0))
      (Variables|Default|SetSpawnedDamageableObject _returnvalue_2)
      (Development|SetHiddeninGame (Variables|Default|GetGhostMesh) true false))))

(event Custom|DoHauntedThings
  (bind _proxymesh (Variables|Default|GetProxyMesh))
  (bind (_origin _boxextent _sphereradius) (Collision|GetComponentBounds _proxymesh))
  (bind _spookysurface (Variables|Default|GetSpookySurface))
  (bind _pfx_hauntedobject (Variables|Default|GetPFX_HauntedObject))
  (bind _ghostmesh (Variables|Default|GetGhostMesh))
  (bind _hauntedsurface (Variables|Default|GetHauntedSurface))
  (bind _instanceparameters (Variables|/Script/CascadeRuntime.ParticleSystemComponent|GetInstanceParameters _pfx_hauntedobject))
  (Utilities|FlowControl|Sequence
    (:then_0
      (if (Variables|Default|GetIsHaunted)
        ;[1 1 1] Move Meshes into place
        (Development|PrintString (Utilities|String|Append (Utilities|GetDisplayName self) " HAUNTED") false false)
        (bind _hauntedtransform (CallFunction|GetHauntedTransform))
        (Rendering|SetVisibility (Variables|/Script/Engine.Actor|GetRootComponent (Variables|Default|GetSpawnedDamageableObject)) false true)
        (Transformation|SetWorldTransform (Variables|Default|GetProxyMesh) _hauntedtransform false true)
        (Transformation|SetWorldTransform (Variables|Default|GetGhostMesh) _hauntedtransform false true)
        ;[1 1 1] Turn on Haunted FX and sound
        (Components|Activation|Activate (Variables|Default|GetPFX_HauntedObject) false)
        (Development|SetHiddeninGame _ghostmesh (not (Variables|Default|GetIsHaunted)) false)
        (bind _returnvalue (Rendering|Material|CreateDynamicMaterialInstance _ghostmesh 0 (Variables|Default|GetGhostMat) "None" :!class /Script/Engine.PrimitiveComponent))
        (bind _output_get (Variables|Default|SetHauntedSurface _returnvalue))
        (Rendering|Material|SetScalarParameterValue _output_get "DisplacementAmount" (Variables|Default|GetDisplacement%20Factor) :!class /Script/Engine.MaterialInstanceDynamic)
        ;[1 1 1] apply appropriate mesh to proxy and ghost mat
        (bind _asstatic_mesh_component (Utilities|Casting|CastToStaticMeshComponent (Variables|/Script/Engine.Actor|GetRootComponent (Variables|Default|GetSpawnedDamageableObject))))
        (Components|StaticMesh|SetStaticMesh (Variables|Default|GetProxyMesh) (Variables|/Script/Engine.StaticMeshComponent|GetStaticMesh _asstatic_mesh_component))
        (Utilities|FlowControl|Sequence
          (:then_0
            ;[1 1 1] Grab Haunted Object material and use that to apply the correct 'floaty' material
            (bind _asmaterial_instance (Utilities|Casting|CastToMaterialInstance (Utilities|Array|Get(acopy) (Rendering|Material|GetMaterials _asstatic_mesh_component))))
            (for _array_element (Variables|/Script/Engine.MaterialInstance|GetTextureParameterValues _asmaterial_instance)
              (Rendering|Material|SetTextureParameterValue _spookysurface (Utilities|Struct|BreakMaterialParameterInfo (Utilities|Struct|BreakTextureParameterValue _array_element)) (.ParameterValue (Utilities|Struct|BreakTextureParameterValue _array_element))))
            (for _array_element_2 (Variables|/Script/Engine.MaterialInstance|GetScalarParameterValues _asmaterial_instance)
              (Rendering|Material|SetScalarParameterValue _spookysurface (Utilities|Struct|BreakMaterialParameterInfo (Utilities|Struct|BreakScalarParameterValue _array_element_2)) (.ParameterValue (Utilities|Struct|BreakScalarParameterValue _array_element_2)) :!class /Script/Engine.MaterialInstanceDynamic))
            (for _array_element_3 (Variables|/Script/Engine.MaterialInstance|GetVectorParameterValues _asmaterial_instance)
              (Rendering|Material|SetVectorParameterValue _spookysurface (Utilities|Struct|BreakMaterialParameterInfo (Utilities|Struct|BreakVectorParameterValue _array_element_3)) (.ParameterValue (Utilities|Struct|BreakVectorParameterValue _array_element_3)) :!class /Script/Engine.MaterialInstanceDynamic)))
          (:then_1
            (Components|StaticMesh|SetStaticMesh (Variables|Default|GetGhostMesh) (Variables|/Script/Engine.StaticMeshComponent|GetStaticMesh _asstatic_mesh_component))
            ;[1 1 1] Make spooky
            (bind _output_get_2 (Variables|Default|SetParticle%20Spawn%20Location _origin))
            (Transformation|SetWorldLocationAndRotation _pfx_hauntedobject _output_get_2 "0, 0, 0" false false)
            (Variables|Default|SetGhost%20Particle%20Parameters _instanceparameters)
            (Utilities|Array|Add _instanceparameters (Utilities|Struct|MakeParticleSysParam "IntLoc" "PSPT_VectorRand" _boxextent (* _boxextent -1.0)))
            (Utilities|Array|Add _instanceparameters (Utilities|Struct|MakeParticleSysParam "SizeScale" "PSPT_Vector" (Math|Vector|MakeVector _sphereradius _sphereradius _sphereradius)))
            (Variables|/Script/CascadeRuntime.ParticleSystemComponent|SetInstanceParameters _instanceparameters _pfx_hauntedobject)
            ;[1 1 1] Adjust strength of bobbing
            (Rendering|Material|SetScalarParameterValue _hauntedsurface "BobSpeed" (Variables|Default|GetSpookyMovement) :!class /Script/Engine.MaterialInstanceDynamic)
            (Rendering|Material|SetVectorParameterValue _hauntedsurface "PivotLoc" (Math|Conversions|ToLinearColor(Vector) (Math|Random|RandomUnitVector)) :!class /Script/Engine.MaterialInstanceDynamic)
            ;[1 1 1] Replaces Object Location for Speed Settings on Bob
            (Rendering|Material|SetScalarParameterValue (Variables|Default|GetHauntedSurface) "RandomValue" (Math|Random|RandomFloatinRange -0.1 0.2) :!class /Script/Engine.MaterialInstanceDynamic)))
        (else
          ;[1 1 1] Turn off Spooky Stuff
          (Rendering|SetVisibility (Variables|Default|GetProxyMesh) false false)
          (Components|Activation|Deactivate (Variables|Default|GetPFX_HauntedObject)) ; Turn off effect
        )))
    (:then_1
      (EventDispatchers|BindEventtoOnDied (Variables|Default|GetSpawnedDamageableObject) (delegate Custom|Event%20On%20Died))
      (Rendering|Material|SetScalarParameterValue (Variables|Default|GetHauntedSurface) "Haunting" 1.0 :!class /Script/Engine.MaterialInstanceDynamic))))

(event Custom|HealthCheck
  ;[1 0.9142 0.2333] Haunted life check.  If health of base object hits 0, clean the haunting up.
  (if (<= (Health|GetHealth (Variables|Default|GetSpawnedDamageableObject) :!class /Script/FortniteGame.BuildingActor) 0.0)
    (Development|SetHiddeninGame (Variables|Default|GetGhostMesh) true false)))

(event Custom|PlayDestructionFX
  ;[0.303 0 0] If we're here, we broke a haunted object
  (Components|Activation|Deactivate (Variables|Default|GetPFX_HauntedObject)) ; Turn off effect
  (Audio|PlaySoundatLocation (Variables|Default|GetGhostEscapeSound) (Transformation|GetActorLocation)) ; Sound on complete
  (Effects|SpawnEmitteratLocation "/Game/Athena/Items/QuestInteractables/HauntedObjects/Effects/PFX_HauntedObject_Destruction.PFX_HauntedObject_Destruction" (Variables|Default|GetParticle%20Spawn%20Location))
  (Development|SetHiddeninGame (Variables|Default|GetProxyMesh) true false))

(event Custom|Event%20On%20Died
  (params (DamagedActor :type Actor%20Object%20Reference) (Damage :type Float%20%28single-precision%29) (InstigatedBy :type Controller%20Object%20Reference) (DamageCauser :type Actor%20Object%20Reference) (HitLocation :type Vector) (FHitComponent :type Primitive%20Component%20Object%20Reference) (BoneName :type Name) (Momentum :type Vector))
  (bind _returnvalue (Fortnite|GetFortPlayerControllerfromActor InstigatedBy))
  ;[1 1 1] Listen for Item to Die
  (Utilities|IsValid _returnvalue
    (:Is%20Valid
      (Quest|SendComplexCustomStatEvent (Quests|GetQuestManager _returnvalue "Invalid") 0 "(GameplayTags=((TagName=\"Athena.Quests.DestroyedHauntedItem\")))" "(GameplayTags=)" 1))
    (:Is%20Not%20Valid
      _))
  (Variables|Default|SetbActive false)
  (Utilities|FlowControl|Delay 0.5)
  (Actor|DestroyActor))

(fn ClientHandleOnRepCalls
  (if (Variables|Default|GetbSpawnedObjectCalled)
    (if (Variables|Default|GetIsHaunted)
      (CallFunction|DoHauntedThings)
      (return)
      (else
        (Development|SetHiddeninGame (Variables|Default|GetProxyMesh) true false)))))

(fn GetHauntedTransform
  (return (Math|Transform|MakeTransform (Math|Vector|MakeVector (.x (.location (Transformation|GetActorTransform))) (.y (.location (Transformation|GetActorTransform))) (+ (.z (.location (Transformation|GetActorTransform))) (Variables|Default|GetFloatingHeight))) (.rotation (Transformation|GetActorTransform)) (.scale (Transformation|GetActorTransform)))))

(fn OnRep_bActive
  (if (not (Variables|Default|GetbActive))
    (Networking|SwitchHasAuthority
      (:Remote
        (CallFunction|PlayDestructionFX)))))

(fn OnRep_IsHaunted
  (Networking|SwitchHasAuthority
    (:Authority
      (if (Variables|Default|GetIsHaunted)
        (CallFunction|CreateChildProp)
        (CallFunction|DoHauntedThings)
        (elif (Variables|Default|GetShouldSpawnChildPropIfNotHaunted)
          (CallFunction|CreateChildProp)
          (else
            (Actor|DestroyActor)))))
    (:Remote
      (CallFunction|ClientHandleonRepCalls))))

(fn OnRep_SpawnedDamageableObject
  (Networking|SwitchHasAuthority
    (:Remote
      (Utilities|IsValid (Variables|Default|GetSpawnedDamageableObject)
        (:Is%20Valid
          (Variables|Default|SetbSpawnedObjectCalled true)
          (CallFunction|ClientHandleonRepCalls))))))

(fn ConstructionScript
  (bind _objectbp (Variables|Default|GetObjectBP))
  (bind _outstaticmeshes (Fortnite|GetStaticMeshesforClass _objectbp))
  (bind _output (Utilities|Array|Get(acopy) _outstaticmeshes))
  (bind _editormesh (Variables|Default|GetEditorMesh))
  (bind _returnvalue (Rendering|Material|CreateDynamicMaterialInstance (Variables|Default|GetProxyMesh) 0 "/Game/Athena/Items/QuestInteractables/HauntedObjects/Materials/SpookyMat.SpookyMat" "None" :!class /Script/Engine.PrimitiveComponent))
  (Variables|Default|SetSpookySurface _returnvalue)
  ;[1 1 1] Gets overridden--this is just so you can preview your work in the editor
  (Components|StaticMesh|SetStaticMesh _editormesh _output)
  (Rendering|Material|SetMaterial _editormesh 0 (StaticMesh|GetMaterial _output 0)))
