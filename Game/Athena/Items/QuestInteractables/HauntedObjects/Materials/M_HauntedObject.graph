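# Translucent lit material for haunted-object quest interactables, giving a wobbling spectral look with animated color.
# Panning 4-way chaos noise displaces vertices along normals and blends two tint colors.
# A static switch adds sine-driven bobbing and rocking, randomized through a ramp texture and disabled on mobile.
# Opacity combines view angle, caustic noise, depth fade and camera fade; the Haunting parameter scales both opacity and noise displacement.
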Material /Game/Athena/Items/QuestInteractables/HauntedObjects/Materials/M_HauntedObject
graph {
 properties {bUsedWithNanite=true, blendMode=BLEND_Translucent, pixelDepthOffsetMode=PDOM_Legacy, usedShadingModels=Default Lit}
 nodes [
  n1 = TextureCoordinate() {uTiling=1.5, vTiling=3}
  n2 = Divide(A=ComponentMask(None=ActorPositionWS()) {g=true, r=true}) {constB=300}
  n3 = MaterialFunctionCall!(
   Coordinates=Panner(Coordinate=Add(A=@n1, B=@n2)) {speedY=0.3}
   Speed
   Divisor
   Texture=TextureObject() {samplerType=SAMPLERTYPE_LinearGrayscale, texture=/Game/Athena/Items/QuestInteractables/HauntedObjects/Textures/T_Smooth_tile}
  ) {materialFunction=/Engine/Functions/Engine_MaterialFunctions03/ComplexMotion/Motion_4WayChaos}
  n4 = Multiply(A=Subtract(A=Multiply(A=@n3.Result) {constB=2}), B=VertexNormalWS())
  n5 = ScalarParameter() {defaultValue=32, parameterName=DisplacementAmount}
  n6 = ScalarParameter() {defaultValue=1, parameterName=Haunting}
  n7 = Constant() {r=-0.01}
  n8 = Constant() {r=0.01}
  n9 = Time()
  n10 = ScalarParameter() {defaultValue=5, parameterName=BobSpeed}
  n11 = TextureSample(UVs=ComponentMask(None=ScalarParameter() {parameterName=RandomValue}) {r=true}) {samplerSource=SSM_Wrap_WorldGroupSettings, samplerType=SAMPLERTYPE_Masks, texture=/Game/Environments/Sets/Arid/Textures/T_TimeRamp}
  n12 = Clamp(None=@n11.R)
  n13 = Sine(None=Multiply(A=Multiply(A=@n9, B=@n10), B=@n12))
  n14 = ObjectPositionWS()
  n15 = WorldPosition()
  n16 = Add(
   A=RotateAboutAxis(
    NormalizedRotationAxis=VectorParameter() {parameterName=PivotLoc, defaultValue=<rgba 1 0 0 1>}.RGB
    RotationAngle=LinearInterpolate(A=@n7, B=@n8, Alpha=@n13)
    PivotPoint=Constant3Vector()
    Position=Subtract(A=@n14, B=@n15.XYZ)
   )
   B=LinearInterpolate(A=Constant3Vector() {constant=<rgba 0 0 1 1>}, B=Constant3Vector() {constant=<rgba 0 0 -2 1>}, Alpha=@n13)
  )
  n17 = Constant()
  Add(
   A=Multiply(A=Multiply(A=@n4, B=@n5), B=@n6)
   B=ShadingPathSwitch(Default=StaticSwitchParameter(True=@n16, False=@n17) {parameterName=BeSpooky}, Mobile=@n17) {desc=Migrated from FeatureLevelSwitch}
  )
  n18 = ObjectPositionWS()
  n19 = Constant() {r=55}
  n20 = TextureSample(UVs=ComponentMask(None=Divide(A=@n18, B=@n19)) {g=true}) {samplerSource=SSM_Wrap_WorldGroupSettings, samplerType=SAMPLERTYPE_Masks, texture=&T_TimeRamp}
  Multiply(A=@n12, B=Clamp(None=@n20.R))
  n21 = TextureCoordinate() {vTiling=0.5}
  n22 = MaterialFunctionCall(
   Coordinates=Panner(Coordinate=Add(A=@n21, B=@n2)) {speedY=0.3}
   Texture=TextureObject() {texture=&T_ShieldCaustics}
  ) {materialFunction=&Motion_4WayChaos}
  Multiply(
   A=LinearInterpolate(
    A=VectorParameter() {parameterName=HauntedColor, defaultValue=<rgba 0.7026 0.1455 1 1>}.RGB
    B=VectorParameter() {parameterName=HauntedColor_1, defaultValue=<rgba 0.7026 0.1455 1 1>}.RGB
    Alpha=@n22.Result
   )
   B=ScalarParameter() {defaultValue=1, parameterName=Amplify}
  )
  n23 = Transform(None=CameraVectorWS()) {transformSourceType=TRANSFORMSOURCE_World, transformType=TRANSFORM_Tangent}
  n24 = Multiply(A=Power(Base=ComponentMask(None=@n23) {b=true}) {constExponent=1.5}, B=@n22.Result)
  n25 = Panner(Coordinate=TextureCoordinate() {uTiling=3, vTiling=1.5}) {speedY=0.5}
  n26 = Constant() {r=0.125}
  n27 = TextureObject() {texture=&T_ShieldCaustics}
  n28 = Multiply(A=Power(Base=MaterialFunctionCall(Coordinates=@n25, Divisor=@n26, Texture=@n27) {materialFunction=&Motion_4WayChaos}.Result) {constExponent=4}) {constB=16}
  n29 = LinearInterpolate(Alpha=Clamp(None=ComponentMask(None=@n28) {r=true})) {constA=48, constB=2}
  n30 = Multiply(
   A=ComponentMask(None=DepthFade(Opacity=@n24, FadeDistance=@n29) {fadeDistanceDefault=32}) {r=true}
   B=MaterialFunctionCall() {materialFunction=&CameraDepthFade}.Result
  )
  n31 = ScalarParameter() {defaultValue=1, parameterName=OpacityAmp}
  Multiply(A=Clamp(None=Multiply(A=@n30, B=@n31)), B=@n6)
 ]
}
